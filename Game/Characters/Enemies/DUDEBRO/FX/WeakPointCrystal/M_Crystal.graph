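# Metallic crystal material for the DUDEBRO enemy's body crystals, covering weak-point and nearby variants via UseWeakpointID.
# Mask textures and UV thresholds split eight body regions, where D1-D8 blend color and emissive to highlighted looks.
# World position offset collapses weak-point crystals toward per-region GrowLoc points via X1-X8, or horn crystals in the other variant, with an alternate UV-encoded-pivot offset.
# AsParticle tints emissive by particle color, otherwise MF_DudebroSkinFX applies.
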Material /Game/Characters/Enemies/DUDEBRO/FX/WeakPointCrystal/M_Crystal
graph {
 properties {
  bUseMaterialAttributes=true
  bUsedWithMeshDeformer=true
  bUsedWithMeshParticles=true
  bUsedWithNanite=true
  bUsedWithNiagaraMeshParticles=true
  bUsedWithSkeletalMesh=true
  ditherOpacityMask=true
  pixelDepthOffsetMode=PDOM_Legacy
  previewMesh=/Engine/Transient.MaterialEditorMeshComponent_23
  usedShadingModels=Default Lit
 }
 nodes [
  Constant3Vector() {constant=<rgba 0.15 0.03 0.0839 1>}
  Constant3Vector() {constant=<rgba 0.0295 0.0565 0.15 1>}
  Constant3Vector() {constant=<rgba 0.2698 0.0353 0.6 1>}
  Constant3Vector() {constant=<rgba 0.0481 0.013 0.175 1>}
  Constant() {r=-600}
  Constant() {r=200}
  Constant() {r=25}
  n1 = CameraVectorWS()
  n2 = VertexNormalWS()
  n3 = Clamp(None=Power(Base=DotProduct(A=@n1, B=@n2), Exp=ScalarParameter() {defaultValue=20, parameterName=ShinePower}))
  n4 = ScalarParameter() {defaultValue=0.1, parameterName=Dark}
  NamedRerouteDeclaration(None=Multiply(A=Add(A=@n3, B=@n4), B=ScalarParameter() {defaultValue=0.5, parameterName=Shine})) {name=Shine, nodeColor=<rgba 1 0 0.9609 1>}
  n5 = NamedRerouteUsage()
  n6 = TextureCoordinate() {coordinateIndex=1}
  n7 = ComponentMask(None=TextureCoordinate() {coordinateIndex=2}) {r=true}
  n8 = Subtract(A=WorldPosition().XYZ, B=TransformPosition(None=AppendVector(A=@n6, B=@n7)) {transformType=TRANSFORMPOSSOURCE_World})
  n9 = TextureCoordinate() {coordinateIndex=2}
  n10 = Transform(None=AppendVector(A=TextureCoordinate() {coordinateIndex=1}, B=ComponentMask(None=@n9) {r=true})) {transformSourceType=TRANSFORMSOURCE_Local}
  n11 = Multiply(A=ComponentMask(None=TextureCoordinate() {coordinateIndex=2}) {g=true}, B=NamedRerouteUsage())
  NamedRerouteDeclaration(None=Add(A=Multiply(A=@n5, B=@n8), B=Multiply(A=@n10, B=@n11))) {name=WPO_1, nodeColor=<rgba 0.5833 0.0231 0.0398 1>}
  n12 = WorldPosition() {worldPositionShaderOffset=WPT_ExcludeAllShaderOffsets}
  n13 = VectorParameter() {parameterName=X1GrowLoc, defaultValue=<rgba 1800 255 947 1>}
  n14 = ComponentMask(None=TextureCoordinate() {coordinateIndex=1}) {g=true}
  n15 = ScalarParameter() {defaultValue=-600, parameterName=bobobo}
  n16 = ScalarParameter() {defaultValue=200, parameterName=hahajeke}
  n17 = Saturate(None=Add(A=Multiply(A=@n14, B=@n15), B=@n16))
  n18 = Multiply(
   A=StaticSwitchParameter(True=Constant() {r=1}, False=Constant()) {parameterName=UseWeakpointID}
   B=TextureSample(UVs=TextureCoordinate() {coordinateIndex=1}) {samplerType=SAMPLERTYPE_Masks, texture=/Game/Characters/Enemies/DUDEBRO/Textures/T_DudeBro_AtkCrystalMask}.RGB
  )
  n19 = MaterialFunctionCall(Float3=Multiply(A=@n17, B=@n18)) {materialFunction=&BreakOutFloat3Components}
  n20 = Multiply(
   A=Multiply(A=Subtract(A=@n12.XYZ, B=@n13.RGB), B=Constant() {r=-1})
   B=Multiply(A=ScalarParameter() {defaultValue=1, desc=Horn-R, parameterName=X1}, B=@n19.R)
  )
  n21 = VectorParameter() {parameterName=X2GrowLoc, defaultValue=<rgba 1795 -181 970 1>}
  n22 = Multiply(
   A=Multiply(A=Subtract(A=@n12.XYZ, B=@n21.RGB), B=Constant() {r=-1})
   B=Multiply(A=ScalarParameter() {defaultValue=1, desc=Horn-L, parameterName=X2}, B=@n19.B)
  )
  n23 = Multiply(A=Subtract(A=@n12.XYZ, B=VectorParameter() {parameterName=X3GrowLoc, defaultValue=<rgba 280 141 872 1>}.RGB), B=Constant() {r=-1})
  n24 = Multiply(A=ScalarParameter() {defaultValue=1, desc=Chest, parameterName=X3}, B=@n19.G)
  n25 = Subtract(A=@n12.XYZ, B=VectorParameter() {parameterName=X4GrowLoc, defaultValue=<rgba -340 22 876 1>}.RGB)
  n26 = Constant() {r=-1}
  n27 = ScalarParameter() {defaultValue=1, desc=Back, parameterName=X4}
  n28 = OneMinus(None=@n14)
  n29 = Saturate(None=Add(A=Multiply(A=@n28, B=@n15), B=@n16))
  n30 = Add(A=@n29, B=@n17)
  n31 = MaterialFunctionCall(Float3=Multiply(A=OneMinus(None=@n30), B=@n18)) {materialFunction=&BreakOutFloat3Components}
  n32 = Add(A=Add(A=Add(A=@n20, B=@n22), B=Multiply(A=@n23, B=@n24)), B=Multiply(A=Multiply(A=@n25, B=@n26), B=Multiply(A=@n27, B=@n31.G)))
  n33 = VectorParameter() {parameterName=X5GrowLoc, defaultValue=<rgba -86 336 824 1>}
  n34 = Multiply(
   A=Multiply(A=Subtract(A=@n12.XYZ, B=@n33.RGB), B=Constant() {r=-1})
   B=Multiply(A=ScalarParameter() {defaultValue=1, desc=Side-R, parameterName=X5}, B=@n31.R)
  )
  n35 = Multiply(A=Subtract(A=@n12.XYZ, B=VectorParameter() {parameterName=X6GrowLoc, defaultValue=<rgba 20 -191 824 1>}.RGB), B=Constant() {r=-1})
  n36 = Multiply(A=ScalarParameter() {defaultValue=1, desc=Side-L, parameterName=X6}, B=@n31.B)
  n37 = Subtract(A=@n12.XYZ, B=VectorParameter() {parameterName=X7GrowLoc, defaultValue=<rgba -360 838 655 1>}.RGB)
  n38 = Constant() {r=-1}
  n39 = ScalarParameter() {defaultValue=1, desc=Arm-R, parameterName=X7}
  n40 = Multiply(A=@n29, B=@n18)
  n41 = ComponentMask(None=@n40) {r=true}
  n42 = Add(A=Add(A=Add(A=@n32, B=@n34), B=Multiply(A=@n35, B=@n36)), B=Multiply(A=Multiply(A=@n37, B=@n38), B=Multiply(A=@n39, B=@n41)))
  n43 = VectorParameter() {parameterName=X8GrowLoc, defaultValue=<rgba 35 -778 403 1>}
  n44 = ComponentMask(None=@n40) {b=true}
  n45 = Multiply(
   A=Multiply(A=Subtract(A=@n12.XYZ, B=@n43.RGB), B=Constant() {r=-1})
   B=Multiply(A=ScalarParameter() {defaultValue=1, desc=Arm-L, parameterName=X8}, B=@n44)
  )
  n46 = ComponentMask(None=TextureCoordinate()) {r=true}
  n47 = Constant() {r=-100}
  n48 = Saturate(None=Add(A=Multiply(A=@n46, B=@n47), B=Constant() {r=50}))
  n49 = Constant()
  n50 = Constant() {r=1}
  n51 = MaterialFunctionCall(Float3=Multiply(
    A=StaticSwitchParameter(True=@n49, False=@n50) {parameterName=UseWeakpointID}
    B=TextureSample() {samplerType=SAMPLERTYPE_Masks, texture=/Game/Characters/Enemies/DUDEBRO/Textures/T_DudeBro_StaticCrystalMask}.RGB
   )) {materialFunction=&BreakOutFloat3Components}
  n52 = Multiply(A=@n48, B=@n51.B)
  n53 = Multiply(A=ComponentMask(None=TextureCoordinate()) {g=true}, B=Constant() {r=-100})
  n54 = Saturate(None=Add(A=@n53, B=Constant() {r=17}))
  n55 = Multiply(A=WorldPosition() {worldPositionShaderOffset=WPT_ExcludeAllShaderOffsets}.XYZ, B=Constant() {r=-1})
  n56 = Multiply(A=Multiply(A=Multiply(A=@n54, B=@n52), B=ScalarParameter() {parameterName=Horn-R-Crystals}), B=@n55)
  n57 = OneMinus(None=@n48)
  n58 = Multiply(A=@n51.B, B=@n57)
  n59 = Multiply(A=@n55, B=Multiply(A=Multiply(A=@n54, B=@n58), B=ScalarParameter() {parameterName=Horn-L-Crystals}))
  NamedRerouteDeclaration(None=StaticSwitchParameter(True=Add(A=@n42, B=@n45), False=Add(A=@n56, B=@n59)) {parameterName=UseWeakpointID}) {name=WPO, nodeColor=<rgba 1 0.0682 0.0395 1>}
  n60 = DynamicParameter() {defaultValue=<rgba 1 1 1 0>, paramNames [Scale,Animation,Param3,Param4]}
  NamedRerouteDeclaration(None=@n60.Animation) {name=DynamicAnimation, nodeColor=<rgba 0 1 0 1>}
  NamedRerouteDeclaration(None=@n60.Scale) {name=DynamicScale, nodeColor=<rgba 1 0 0 1>}
  n61 = Reroute()
  n62 = Reroute()
  ShadingPathSwitch(Default=FeatureLevelSwitch(Default=@n61, ES2=@n62, ES3_1=@n62, SM4_REMOVED=@n61), Forward=@n62, Mobile=@n62)
  n63 = ScalarParameter() {defaultValue=0.15, parameterName=RRR-A}
  n64 = ScalarParameter() {defaultValue=0.4, parameterName=RRR-B}
  n65 = TextureSample(UVs=Multiply(A=TextureCoordinate(), B=ScalarParameter() {defaultValue=2, parameterName=RoughTile})) {samplerType=SAMPLERTYPE_Masks, texture=/Game/Global/Textures/Default/Noise/T_CubeNoise}
  n66 = Multiply(A=@n65.R, B=@n65.R)
  n67 = ScalarParameter() {parameterName=UseAddShineColor}
  n68 = VectorParameter() {parameterName=UV, defaultValue=<rgba 8 8 0 0>}
  n69 = TextureSample(UVs=Multiply(A=TextureCoordinate(), B=AppendVector(A=@n68.R, B=@n68.R))) {samplerType=SAMPLERTYPE_Normal, texture=/Game/FrontEnd/Store/Cards/Assets/T_Crystal_N}
  n70 = Constant3Vector() {constant=<rgba 0 0 1 0>}
  n71 = Constant() {r=0.7}
  n72 = Transform(None=Normalize(VectorInput=LinearInterpolate(A=@n69.RGB, B=@n70, Alpha=@n71)))
  n73 = MaterialFunctionCall(Exponent=Constant() {r=0.25}, Vertex / Pixel Normal=@n72) {materialFunction=&MF_Fresnel_FX}
  n74 = LinearInterpolate(
   A=VectorParameter() {parameterName=Emissive1-WP, defaultValue=<rgba 1.7764 0.04 4 1>}.RGB
   B=VectorParameter() {parameterName=Emissive2-WP, defaultValue=<rgba 0.2013 0 2 1>}.RGB
   Alpha=@n73.Fresnel
  )
  n75 = LinearInterpolate(
   A=VectorParameter() {parameterName=Emissive1-NearbyWP, defaultValue=<rgba 1.899 0 3 1>}.RGB
   B=VectorParameter() {parameterName=Emissive2-NearbyWP, defaultValue=<rgba 0.0527 0.025 0.25 1>}.RGB
   Alpha=@n73.Fresnel
  )
  n76 = TextureSample(UVs=Multiply(A=TextureCoordinate(), B=Constant() {r=10})) {texture=/Game/Effects/Fort_Effects/Textures/Tile/T_Glass_Cracks_M}
  n77 = VectorParameter() {parameterName=Emissive1-Normal, defaultValue=<rgba 0.15 0.03 0.0839 1>}
  n78 = VectorParameter() {parameterName=Emissive2-Normal, defaultValue=<rgba 0.0788 0.0213 0.3 1>}
  n79 = SphereMask(
   A=WorldPosition().XYZ
   B=ObjectPositionWS()
   Radius=ScalarParameter() {defaultValue=2200, parameterName=Obj-Pos-Sphere-Glow-Rad}
   Hardness=Constant() {r=0.25}
  )
  n80 = Power(Base=DotProduct(A=CameraVectorWS(), B=@n72), Exp=Constant() {r=4}) {constExponent=8}
  n81 = Power(Base=@n76.R, Exp=Constant() {r=0.5}) {constExponent=8}
  n82 = Add(
   A=Multiply(A=ScalarParameter() {desc=Horn-R, parameterName=D1}, B=@n19.R)
   B=Multiply(A=ScalarParameter() {desc=Horn-L, parameterName=D2}, B=@n19.B)
  )
  n83 = Saturate(None=Add(A=Constant() {r=74}, B=@n53))
  n84 = Multiply(A=@n83, B=OneMinus(None=@n54))
  n85 = Multiply(A=ScalarParameter() {desc=Chest, parameterName=D3}, B=Add(A=@n19.G, B=Multiply(A=@n84, B=@n51.B)))
  n86 = ScalarParameter() {desc=Back, parameterName=D4}
  n87 = Add(A=@n31.G, B=Multiply(A=OneMinus(None=@n83), B=@n51.B))
  n88 = Multiply(A=@n51.G, B=@n48)
  n89 = Add(
   A=Add(A=Add(A=@n82, B=@n85), B=Multiply(A=@n86, B=@n87))
   B=Multiply(A=ScalarParameter() {desc=Side-R, parameterName=D5}, B=Add(A=@n31.R, B=@n88))
  )
  n90 = Multiply(A=ScalarParameter() {desc=Side-L, parameterName=D6}, B=Add(A=@n31.B, B=Multiply(A=@n51.G, B=@n57)))
  n91 = ScalarParameter() {desc=Arm-R, parameterName=D7}
  n92 = Add(A=@n41, B=Multiply(A=@n51.R, B=@n48))
  n93 = Multiply(A=@n51.R, B=@n57)
  n94 = Add(
   A=Add(A=Add(A=@n89, B=@n90), B=Multiply(A=@n91, B=@n92))
   B=Multiply(A=ScalarParameter() {desc=Arm-L, parameterName=D8}, B=Add(A=@n44, B=@n93))
  )
  n95 = LinearInterpolate(
   A=Multiply(A=Multiply(A=@n80, B=@n81), B=LinearInterpolate(A=@n77.RGB, B=@n78.RGB, Alpha=@n79))
   B=Multiply(A=StaticSwitchParameter(True=@n74, False=@n75) {parameterName=UseWeakpointID}, B=OneMinus(None=@n76.R))
   Alpha=@n94
  )
  n96 = Add(A=@n95, B=NamedRerouteUsage())
  n97 = ScalarParameter() {parameterName=WPO/WPO_1}
  n98 = NamedRerouteUsage()
  n99 = NamedRerouteUsage()
  n100 = Multiply(
   A=MaterialFunctionCall() {materialFunction=&DitherTemporalAA}.Result
   B=ScalarParameter() {defaultValue=18, parameterName=PDO}
  ) {constB=25}
  n101 = Constant()
  n102 = VectorParameter() {parameterName=BaseColor-Normal, defaultValue=<rgba 0.2124 0.2 0.5 1>}
  n103 = StaticSwitchParameter(
   True=VectorParameter() {parameterName=BaseColor-WP, defaultValue=<rgba 1 0.0078 0.8784 1>}.RGB
   False=VectorParameter() {parameterName=BaseColor-NearbyWP, defaultValue=<rgba 0.2529 0.094 1 1>}.RGB
  ) {parameterName=UseWeakpointID}
  n104 = GetMaterialAttributes(None=MakeMaterialAttributes!(
    BaseColor=LinearInterpolate(A=@n102.RGB, B=@n103, Alpha=@n94)
    Metallic=ScalarParameter() {defaultValue=0.9, parameterName=Metal}
    Specular=ScalarParameter() {defaultValue=0.5, parameterName=Spec1}
    Roughness=LinearInterpolate(A=@n63, B=@n64, Alpha=@n66)
    Anisotropy
    EmissiveColor=If(A=@n67, "A > B"=@n96, "A == B"=@n95, "A < B"=@n95)
    Opacity
    OpacityMask
    Normal
    Tangent
    WorldPositionOffset=If(A=@n97, "A > B"=@n98, "A == B"=@n99, "A < B"=@n99)
    SubsurfaceColor
    ClearCoat
    ClearCoatRoughness
    AmbientOcclusion
    Refraction
    CustomizedUVs_0
    CustomizedUVs_1
    CustomizedUVs_2
    CustomizedUVs_3
    CustomizedUVs_4
    CustomizedUVs_5
    CustomizedUVs_6
    CustomizedUVs_7
    PixelDepthOffset=QualitySwitch(Default=@n100, Low=@n101, High=@n100, Medium=@n101)
    ShadingModel
    Displacement
   )) {attributeGetTypes [#1]}
  n105 = ParticleColor()
  StaticSwitchParameter(
   True=SetMaterialAttributes(MaterialAttributes=@n104.MaterialAttributes, Emissive Color=Multiply(A=@n104.Emissive Color, B=@n105.RGB)) {attributeSetTypes [#1]}
   False=MaterialFunctionCall!(InMA=@n104.MaterialAttributes, HaveMasks=StaticBool(), "FresnelRevert(WithOutMask)", EmissiveMutiply) {materialFunction=/Game/Characters/Enemies/DUDEBRO/FX/SkinFX/MF_DudebroSkinFX}.MA
  ) {parameterName=AsParticle}
 ]
}
