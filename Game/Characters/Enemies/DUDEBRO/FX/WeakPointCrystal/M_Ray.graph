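# Translucent, unlit, two-sided mesh-particle material for rays in the DUDEBRO enemy's weak point crystal FX.
# Noise panned by dynamic-parameter time, plus a V-gradient with adjustable exponent, blends three colors, tinted by particle color into emissive adjustments.
# Opacity combines that mask with the gradient, a smooth ramp, particle alpha and view-facing falloff that softens silhouette edges.
# Depth fade applies only on the default shading path, not mobile or forward.
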Material /Game/Characters/Enemies/DUDEBRO/FX/WeakPointCrystal/M_Ray
graph {
 properties {
  bUsedWithMeshParticles=true
  bUsedWithNanite=true
  bUsedWithNiagaraMeshParticles=true
  blendMode=BLEND_Translucent
  pixelDepthOffsetMode=PDOM_Legacy
  shadingModel=MSM_Unlit
  twoSided=true
  usedShadingModels=Unlit
 }
 nodes [
  n1 = ParticleColor()
  n2 = VectorParameter() {parameterName=ColorA, defaultValue=<rgba 2 0 1.0006 1>}
  n3 = VectorParameter() {parameterName=ColorB, defaultValue=<rgba 0.2012 0 0.8542 1>}
  n4 = VectorParameter() {parameterName=ColorC, defaultValue=<rgba 0 0.0568 1 1>}
  n5 = ScalarParameter() {parameterName=MinValue}
  n6 = TextureCoordinate()
  n7 = VectorParameter() {parameterName=UVandSpeed, defaultValue=<rgba 1 1 0.2 0>}
  n8 = AppendVector(A=@n7.R, B=@n7.R)
  n9 = DynamicParameter() {defaultValue=<rgba 1 1 2 0>, paramNames [Time, Param2, AlphaExp, "AlphaOffset(AfterExp)"]}
  n10 = TextureSampleParameter2D(UVs=Panner(Coordinate=Multiply(A=@n6, B=@n8), Time=@n9.Time, Speed=AppendVector(A=@n7.B, B=@n7.B))) {parameterName=Noise, samplerType=SAMPLERTYPE_LinearColor, texture=&T_NoisePack_LinearColor}
  n11 = TextureCoordinate()
  n12 = OneMinus(None=ComponentMask(None=@n11) {g=true})
  n13 = Power(Base=@n12, Exp=@n9.AlphaExp)
  n14 = Add(A=Multiply(A=LinearInterpolate(A=@n5, Alpha=@n10.R), B=ScalarParameter() {parameterName=BlendWithAlpha}), B=@n13)
  MaterialFunctionCall(Emissive=Multiply(
    A=@n1.RGB
    B=MaterialFunctionCall(A=@n2.RGB, B=@n3.RGB, C=@n4.RGB, Alpha=@n14) {materialFunction=&3ColorBlend}.Result
   )) {materialFunction=&MF_FX_EmissiveAdjustments}
  n15 = DotProduct(A=VertexNormalWS(), B=CameraVectorWS())
  n16 = Add(A=@n9."AlphaOffset(AfterExp)", B=@n14)
  n17 = MaterialFunctionCall(XYZ=@n11, XY=@n11) {materialFunction=&Swizzle}
  n18 = Multiply(A=@n13, B=Multiply(A=@n12, B=TextureSample(UVs=@n17.YX) {texture=/Game/Effects/Fort_Effects/Textures/Ramps/T_Ramp_Smooth}.R))
  n19 = Multiply(A=@n1.A, B=Multiply(A=Abs(None=@n15), B=Multiply(A=@n16, B=@n18)))
  n20 = DepthFade(FadeDistance=ScalarParameter() {parameterName=DF_FD})
  ShadingPathSwitch(Default=ShadingPathSwitch(Default=Multiply(A=@n19, B=@n20), Mobile=@n19) {desc=Migrated from FeatureLevelSwitch}, Forward=@n19, Mobile=@n19)
 ]
}
