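# Translucent unlit mesh-particle material for energy waves on the DUDEBRO enemy's weak point crystal effect.
# Two world-space flowmaps distort noise and smoke textures, shaped by squared vertex red and particle alpha into opacity, which drives a three-color gradient.
# Non-mobile paths blend that gradient toward tinted scene color.
# Dynamic parameters drive panning and noisy normal-based vertex offset, halved for the previous frame to affect motion vectors.
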Material /Game/Characters/Enemies/DUDEBRO/FX/WeakPointCrystal/M_Waves
graph {
 properties {
  bUseMaterialAttributes=true
  bUsedWithMeshParticles=true
  bUsedWithNiagaraMeshParticles=true
  blendMode=BLEND_Translucent
  pixelDepthOffsetMode=PDOM_Legacy
  shadingModel=MSM_Unlit
  usedShadingModels=Unlit
 }
 nodes [
  n1 = DynamicParameter() {defaultValue=<rgba 1 1 1 50>, paramNames [Time,Param2,WPO_Scale,WPO_Center]}
  n2 = Add(A=Multiply(A=VertexColor().R, B=@n1.WPO_Center), B=@n1.WPO_Scale)
  n3 = TextureSample(UVs=Panner(
    Coordinate=TextureCoordinate() {uTiling=3, vTiling=0}
    Time=DynamicParameter() {defaultValue=<rgba 1 1 1 50>, paramNames [Time,Param2,WPO_Scale,WPO_Center]}.Time
   ) {speedY=0.4}) {samplerType=SAMPLERTYPE_LinearColor, texture=&T_NoisePack_LinearColor}
  n4 = Multiply(A=VertexNormalWS(), B=Multiply(A=LinearInterpolate(Alpha=@n3.G) {constA=0.8, constB=1.2}))
  n5 = PreviousFrameSwitch(Current Frame=@n4, Previous Frame=Multiply(A=@n4) {constB=0.5})
  n6 = ParticleColor()
  n7 = AppendVector(A=VertexColor().R, B=ComponentMask(None=TextureCoordinate()) {r=true})
  n8 = VectorParameter() {parameterName=UV, defaultValue=<rgba 0.5 8 1.5 0>}
  n9 = AppendVector(A=@n8.R, B=@n8.R)
  n10 = Divide(A=ComponentMask(None=WorldPosition().XYZ) {g=true, r=true}, B=ScalarParameter() {defaultValue=32768, parameterName=WorldPosScale}) {constB=4096}
  n11 = Multiply(A=@n10, B=ScalarParameter() {defaultValue=12, parameterName=FlowmapSize})
  n12 = Multiply(A=Time(), B=ScalarParameter() {defaultValue=0.005, parameterName=TTTT})
  n13 = Panner(
   Coordinate=Multiply(A=@n10, B=ScalarParameter() {defaultValue=35, parameterName=FlowmapSize2})
   Time=Multiply(A=Time(), B=ScalarParameter() {defaultValue=-0.0125, parameterName=TTTT2})
  ) {speedX=-10, speedY=5}
  n14 = Add(
   A=TextureSample(UVs=Panner(Coordinate=@n11, Time=@n12) {speedX=-10, speedY=5}) {samplerType=SAMPLERTYPE_LinearColor, texture=/Game/VisualThreatManager/StormVisuals/LargeClouds/T_Swirl_Flowmap}.RGB
   B=Multiply(
    A=TextureSample(UVs=@n13) {samplerType=SAMPLERTYPE_LinearColor, texture=&T_Swirl_Flowmap}.RGB
    B=ScalarParameter() {defaultValue=0.3, parameterName=FlowmapPower2}
   )
  )
  n15 = Multiply(A=Subtract(A=ComponentMask(None=@n14) {g=true, r=true}) {constB=0.5}, B=ScalarParameter() {defaultValue=0.5, parameterName=FlowmapPower})
  n16 = Add(
   A=Panner(
    Coordinate=Multiply(A=@n7, B=@n9)
    Time=DynamicParameter() {defaultValue=<rgba 1 1 1 50>, paramNames [Time,Param2,WPO_Scale,WPO_Center]}.Time
    Speed=AppendVector(A=@n8.B, B=@n8.B)
   )
   B=@n15
  )
  n17 = Sine(None=ComponentMask(None=TextureCoordinate()) {g=true})
  n18 = VertexColor()
  n19 = Multiply(A=@n18.R, B=@n18.R)
  n20 = Multiply(
   A=Multiply(
    A=TextureSample(UVs=@n16) {samplerType=SAMPLERTYPE_LinearColor, texture=&T_NoisePack_LinearColor}.G
    B=Abs(None=@n17)
   )
   B=@n19
  )
  n21 = TextureSampleParameter2D(UVs=Add(A=@n15, B=Multiply(A=@n10) {constB=2})) {parameterName=Texture, texture=/Game/Global/Textures/Default/Tileable_Patterns/T_Smoke_OUT}
  n22 = Subtract(A=Multiply(A=@n6.A, B=Multiply(A=@n20, B=@n21.R))) {constB=0.02}
  n23 = Multiply(
   A=MaterialFunctionCall(
    A=VectorParameter() {parameterName=52rtwswe, defaultValue=<rgba 0.0083 0.0167 0.0469 1>}.RGB
    B=VectorParameter() {parameterName=KKKRTEe, defaultValue=<rgba 0.1433 0.0355 0.6823 1>}.RGB
    C=VectorParameter() {parameterName=Dwet34wr2, defaultValue=<rgba 2 0.2 0.6287 1>}.RGB
    Alpha=@n22
   ) {materialFunction=&3ColorBlend}.Result
   B=@n6.RGB
  )
  n24 = Multiply(A=Multiply(A=SceneColor!(UVs).RGB, B=VectorParameter() {parameterName=SceneColorTint, defaultValue=<rgba 10 1 1.15 1>}.RGB), B=@n19)
  SetMaterialAttributes!(
   MaterialAttributes=SetMaterialAttributes!(MaterialAttributes, World Position Offset=Multiply(A=@n2, B=@n5)) {attributeSetTypes [#1]}
   Emissive Color=ShadingPathSwitch(Default=LinearInterpolate(A=@n23, B=@n24, Alpha=@n19), Mobile=@n23) {desc=Migrated from FeatureLevelSwitch}
   Opacity=@n22
  ) {attributeSetTypes [#2, #3]}
 ]
}
